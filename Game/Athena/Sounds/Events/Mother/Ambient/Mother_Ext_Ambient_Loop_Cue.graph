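# Looping exterior ambience cue for the Mother event, pairing an ambient bed with a distant battle loop.
# Each loop has a short volume fade-in envelope before mixing, and the battle layer is scaled by a Skyfire parameter.
# A Fade parameter scales the whole mix, and both parameters default to zero.
# Overrides disable distance attenuation, enable occlusion low-pass filtering and allow only one instance, rejecting new ones.
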SoundCue /Game/Athena/Sounds/Events/Mother/Ambient/Mother_Ext_Ambient_Loop_Cue
root=ModulatorContinuous_0
nodes [
 {
  id=Enveloper_0
  children [WavePlayer_0]
  properties {pitchCurve {editorCurveData {keys [{time=0}]}}, volumeCurve {editorCurveData {keys [<at 0 0 RCIM_Cubic>,<at 1 1>]}}}
 }
 {
  id=Enveloper_1
  children [WavePlayer_1]
  properties {pitchCurve {editorCurveData {keys [{time=0}]}}, volumeCurve {editorCurveData {keys [<at 0 0 RCIM_Cubic>,<at 1 1>]}}}
 }
 {id=Mixer_0, children [Enveloper_0, ModulatorContinuous_1], properties {inputVolume [1,1]}}
 {id=ModulatorContinuous_0, children [Mixer_0], properties {volumeModulationParams {default=0, parameterName=Fade}}}
 {id=ModulatorContinuous_1, children [Enveloper_1], properties {volumeModulationParams {default=0, parameterName=Skyfire}}}
 {id=WavePlayer_0, properties {bLooping=true, soundWaveAssetPtr=&/Mother_Ext_Ambient_Loop}}
 {
  id=WavePlayer_1
  properties {bLooping=true, soundWaveAssetPtr=&/Mother_Ext_Ambient_Distant_Battle_Loop}
 }
]
properties {
 attenuationSettings=&/Mother_Ext_Ambient_Loop_Att
 bOverrideAttenuation=true
 bOverrideConcurrency=true
 maxDistance=2097152
 soundClassObject=&Fort_SFX_InGame
 volumeMultiplier=1
 attenuationOverrides {
  bAttenuate=false
  bEnableFocusInterpolation=true
  bEnableOcclusion=true
  falloffDistance=0
  focusAzimuth=10
  nonFocusAzimuth=30
  nonFocusDistanceScale=1.5
  nonFocusVolumeAttenuation=0
  occlusionInterpolationTime=2
  occlusionLowPassFilterFrequency=500
  occlusionTraceChannel=ECC_GameTraceChannel11
  occlusionVolumeAttenuation=0
  attenuationShapeExtents=<220000 220000 100>
 }
 concurrencyOverrides {maxCount=1, resolutionRule=PreventNew}
}
