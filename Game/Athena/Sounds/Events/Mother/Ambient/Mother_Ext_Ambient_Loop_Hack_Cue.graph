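# Sound Cue for the Mother event's exterior ambient loop, named as a hack, that plays only a looping five-second silence wave.
# A single looping wave player is the root, routed to the in-game SFX sound class.
# Inline attenuation overrides replace the referenced attenuation asset, setting a huge extent with zero falloff, listener focus and low-pass occlusion.
# Inline concurrency allows one instance and prevents new ones.
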SoundCue /Game/Athena/Sounds/Events/Mother/Ambient/Mother_Ext_Ambient_Loop_Hack_Cue
root=WavePlayer_1
nodes [{id=WavePlayer_1, properties {bLooping=true, soundWaveAssetPtr=/Game/Sounds/Fort_Test_Sounds/5_seconds_silence}}]
properties {
 attenuationSettings=&/Mother_Ext_Ambient_Loop_Att
 bOverrideAttenuation=true
 bOverrideConcurrency=true
 maxDistance=2097152
 soundClassObject=&Fort_SFX_InGame
 volumeMultiplier=1
 attenuationOverrides {
  bEnableListenerFocus=true
  bEnableOcclusion=true
  falloffDistance=0
  focusAzimuth=10
  nonFocusAzimuth=30
  nonFocusDistanceScale=100
  nonFocusVolumeAttenuation=0
  occlusionInterpolationTime=2
  occlusionLowPassFilterFrequency=500
  occlusionTraceChannel=ECC_GameTraceChannel11
  occlusionVolumeAttenuation=0
  attenuationShapeExtents=<10000000 220000 125000>
 }
 concurrencyOverrides {maxCount=1, resolutionRule=PreventNew}
}
