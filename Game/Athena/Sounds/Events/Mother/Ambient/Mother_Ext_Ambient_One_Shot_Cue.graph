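# Sound cue that plays a random exterior ambient one-shot for Mother, a Fortnite Athena event.
# The first two quality slots pick from eleven mixes, each layering a stereo wave through an attenuation node with its matching quad wave.
# The other two slots use a lighter set of four attenuated stereo waves, leaving the quad inputs empty.
# A concurrency override allows one instance at a time and blocks new triggers.
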SoundCue /Game/Athena/Sounds/Events/Mother/Ambient/Mother_Ext_Ambient_One_Shot_Cue
root=QualityLevel_1
nodes [
 {
  id=Attenuation_0
  children [WavePlayer_31]
  properties {attenuationSettings=&/Mother_Ext_Ambient_One_Shot_Att}
 }
 {
  id=Attenuation_1
  children [WavePlayer_33]
  properties {attenuationSettings=&Mother_Ext_Ambient_One_Shot_Att}
 }
 {
  id=Attenuation_10
  children [WavePlayer_30]
  properties {attenuationSettings=&Mother_Ext_Ambient_One_Shot_Att}
 }
 {
  id=Attenuation_13
  children [WavePlayer_50]
  properties {attenuationSettings=&Mother_Ext_Ambient_One_Shot_Att}
 }
 {
  id=Attenuation_14
  children [WavePlayer_51]
  properties {attenuationSettings=&Mother_Ext_Ambient_One_Shot_Att}
 }
 {
  id=Attenuation_15
  children [WavePlayer_52]
  properties {attenuationSettings=&Mother_Ext_Ambient_One_Shot_Att}
 }
 {
  id=Attenuation_17
  children [WavePlayer_55]
  properties {attenuationSettings=&Mother_Ext_Ambient_One_Shot_Att}
 }
 {
  id=Attenuation_2
  children [WavePlayer_17]
  properties {attenuationSettings=&Mother_Ext_Ambient_One_Shot_Att}
 }
 {
  id=Attenuation_3
  children [WavePlayer_18]
  properties {attenuationSettings=&Mother_Ext_Ambient_One_Shot_Att}
 }
 {
  id=Attenuation_4
  children [WavePlayer_20]
  properties {attenuationSettings=&Mother_Ext_Ambient_One_Shot_Att}
 }
 {
  id=Attenuation_5
  children [WavePlayer_23]
  properties {attenuationSettings=&Mother_Ext_Ambient_One_Shot_Att}
 }
 {
  id=Attenuation_6
  children [WavePlayer_24]
  properties {attenuationSettings=&Mother_Ext_Ambient_One_Shot_Att}
 }
 {
  id=Attenuation_7
  children [WavePlayer_27]
  properties {attenuationSettings=&Mother_Ext_Ambient_One_Shot_Att}
 }
 {
  id=Attenuation_8
  children [WavePlayer_28]
  properties {attenuationSettings=&Mother_Ext_Ambient_One_Shot_Att}
 }
 {
  id=Attenuation_9
  children [WavePlayer_29]
  properties {attenuationSettings=&Mother_Ext_Ambient_One_Shot_Att}
 }
 {id=Mixer_11, children [Attenuation_13, null], properties {inputVolume [1,1]}}
 {id=Mixer_37, children [Attenuation_10, WavePlayer_13], properties {inputVolume [1,1]}}
 {id=Mixer_38, children [Attenuation_9, WavePlayer_12], properties {inputVolume [1,1]}}
 {id=Mixer_39, children [Attenuation_8, WavePlayer_11], properties {inputVolume [1,1]}}
 {id=Mixer_40, children [Attenuation_7, WavePlayer_9], properties {inputVolume [1,1]}}
 {id=Mixer_41, children [Attenuation_6, WavePlayer_7], properties {inputVolume [1,1]}}
 {id=Mixer_42, children [Attenuation_5, WavePlayer_6], properties {inputVolume [1,1]}}
 {id=Mixer_43, children [Attenuation_4, WavePlayer_4], properties {inputVolume [1,1]}}
 {id=Mixer_44, children [Attenuation_3, WavePlayer_3], properties {inputVolume [1,1]}}
 {id=Mixer_45, children [Attenuation_2, WavePlayer_1], properties {inputVolume [1,1]}}
 {id=Mixer_46, children [Attenuation_1, WavePlayer_16], properties {inputVolume [1,1]}}
 {id=Mixer_47, children [Attenuation_0, WavePlayer_15], properties {inputVolume [1,1]}}
 {id=Mixer_6, children [Attenuation_17, null], properties {inputVolume [1,1]}}
 {id=Mixer_8, children [Attenuation_15, null], properties {inputVolume [1,1]}}
 {id=Mixer_9, children [Attenuation_14, null], properties {inputVolume [1,1]}}
 {id=QualityLevel_1, children [Random_4,Random_4,Random_1,Random_1]}
 {id=Random_1, children [Mixer_11,Mixer_9,Mixer_8,Mixer_6], properties {weights [1,1,1,1]}}
 {
  id=Random_4
  children [Mixer_47,Mixer_46,Mixer_45,Mixer_44,Mixer_43,Mixer_42,Mixer_41,Mixer_40,Mixer_39,Mixer_38,Mixer_37]
  properties {weights [1,1,1,1,1,1,1,1,1,1,1]}
 }
 {id=WavePlayer_1, properties {soundWaveAssetPtr=&/Mother_Ext_Ambient_One_Shot_Quad_11}}
 {id=WavePlayer_11, properties {soundWaveAssetPtr=&/Mother_Ext_Ambient_One_Shot_Quad_30}}
 {id=WavePlayer_12, properties {soundWaveAssetPtr=&/Mother_Ext_Ambient_One_Shot_Quad_31}}
 {id=WavePlayer_13, properties {soundWaveAssetPtr=&/Mother_Ext_Ambient_One_Shot_Quad_34}}
 {id=WavePlayer_15, properties {soundWaveAssetPtr=&/Mother_Ext_Ambient_One_Shot_Quad_01}}
 {id=WavePlayer_16, properties {soundWaveAssetPtr=&/Mother_Ext_Ambient_One_Shot_Quad_07}}
 {id=WavePlayer_17, properties {soundWaveAssetPtr=&/Mother_Ext_Ambient_One_Shot_Stereo_11}}
 {id=WavePlayer_18, properties {soundWaveAssetPtr=&/Mother_Ext_Ambient_One_Shot_Stereo_14}}
 {id=WavePlayer_20, properties {soundWaveAssetPtr=&/Mother_Ext_Ambient_One_Shot_Stereo_17}}
 {id=WavePlayer_23, properties {soundWaveAssetPtr=&/Mother_Ext_Ambient_One_Shot_Stereo_19}}
 {id=WavePlayer_24, properties {soundWaveAssetPtr=&/Mother_Ext_Ambient_One_Shot_Stereo_23}}
 {id=WavePlayer_27, properties {soundWaveAssetPtr=&/Mother_Ext_Ambient_One_Shot_Stereo_28}}
 {id=WavePlayer_28, properties {soundWaveAssetPtr=&/Mother_Ext_Ambient_One_Shot_Stereo_30}}
 {id=WavePlayer_29, properties {soundWaveAssetPtr=&/Mother_Ext_Ambient_One_Shot_Stereo_31}}
 {id=WavePlayer_3, properties {soundWaveAssetPtr=&/Mother_Ext_Ambient_One_Shot_Quad_14}}
 {id=WavePlayer_30, properties {soundWaveAssetPtr=&/Mother_Ext_Ambient_One_Shot_Stereo_34}}
 {id=WavePlayer_31, properties {soundWaveAssetPtr=&/Mother_Ext_Ambient_One_Shot_Stereo_01}}
 {id=WavePlayer_33, properties {soundWaveAssetPtr=&/Mother_Ext_Ambient_One_Shot_Stereo_07}}
 {id=WavePlayer_4, properties {soundWaveAssetPtr=&/Mother_Ext_Ambient_One_Shot_Quad_17}}
 {id=WavePlayer_50, properties {soundWaveAssetPtr=&Mother_Ext_Ambient_One_Shot_Stereo_11}}
 {id=WavePlayer_51, properties {soundWaveAssetPtr=&Mother_Ext_Ambient_One_Shot_Stereo_14}}
 {id=WavePlayer_52, properties {soundWaveAssetPtr=&Mother_Ext_Ambient_One_Shot_Stereo_17}}
 {id=WavePlayer_55, properties {soundWaveAssetPtr=&Mother_Ext_Ambient_One_Shot_Stereo_23}}
 {id=WavePlayer_6, properties {soundWaveAssetPtr=&/Mother_Ext_Ambient_One_Shot_Quad_19}}
 {id=WavePlayer_7, properties {soundWaveAssetPtr=&/Mother_Ext_Ambient_One_Shot_Quad_23}}
 {id=WavePlayer_9, properties {soundWaveAssetPtr=&/Mother_Ext_Ambient_One_Shot_Quad_28}}
]
properties {
 attenuationSettings=&Mother_Ext_Ambient_One_Shot_Att
 bOverrideConcurrency=true
 maxDistance=2097152
 soundClassObject=&Fort_SFX_InGame
 volumeMultiplier=2
 concurrencyOverrides {maxCount=1, resolutionRule=PreventNew}
}
